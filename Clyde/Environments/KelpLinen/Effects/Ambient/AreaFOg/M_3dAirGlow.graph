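# Additive unlit material for KelpLinen ambient air glow on mesh particles.
# Particle color times Fresnel gets depth fade and optional camera fade, both skipped at low quality, before emissive adjustments; the double-sided Fresnel lerp has identical inputs.
# Opacity branches combine optional particle alpha and desaturated glow with either location fade or camera-to-actor distance fade.
# A switch on Main_Phase_Tracker selects per-phase opacity, with no visible link to those branches.
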Material /Clyde/Environments/KelpLinen/Effects/Ambient/AreaFOg/M_3dAirGlow
graph {
 properties {
  bUsedWithMeshParticles=true
  bUsedWithNanite=true
  bUsedWithNiagaraMeshParticles=true
  blendMode=BLEND_Additive
  pixelDepthOffsetMode=PDOM_Legacy
  shadingModel=MSM_Unlit
  usedShadingModels=Unlit
 }
 nodes [
  n1 = ParticleColor()
  n2 = MaterialFunctionCall(
   Brightness=ScalarParameter() {defaultValue=1, parameterName=Fresnel Brightness}
   Exponent=ScalarParameter() {defaultValue=10, parameterName=Fresnel Exponent}
  ) {materialFunction=&MF_Fresnel_FX}
  n3 = Multiply(A=@n1.RGB, B=LinearInterpolate(A=@n2.Fresnel, B=@n2.Fresnel, Alpha=ScalarParameter() {parameterName=use Double sided Fres}))
  n4 = ScalarParameter() {defaultValue=500, parameterName=CameraFadeDistance}
  n5 = Multiply(
   A=DepthFade(Opacity=@n3, FadeDistance=ScalarParameter() {defaultValue=500, parameterName=Depth Fade Distance})
   B=StaticSwitchParameter(
    True=MaterialFunctionCall(Distance=@n4) {materialFunction=&Near_Camera_Fade}.Result
    False=Constant() {r=1}
   ) {parameterName=UseCameraFade}
  )
  MaterialFunctionCall(Emissive=QualitySwitch(Default=@n5, Low=@n3, High=@n5, Medium=@n5)) {materialFunction=&MF_FX_EmissiveAdjustments}
  n6 = ScalarParameter() {parameterName="Fade Length (S)"}
  n7 = ScalarParameter() {parameterName="Fade Offset (S)"}
  MaterialFunctionCall(Fade Length=@n6, Fade Offset=@n7) {materialFunction=&CameraDepthFade}
  n8 = StaticSwitchParameter(True=@n1.A, False=Constant() {r=1}) {parameterName=UseParticleOpacity?}
  n9 = Desaturation(None=@n5, Fraction=Constant() {r=1})
  n10 = StaticSwitchParameter(True=Multiply(A=@n8, B=ComponentMask(None=@n9) {r=true}), False=@n8) {parameterName=MultiplyColorAgainstOpacity}
  Multiply(A=@n10, B=OneMinus(None=MaterialFunctionCall() {materialFunction=/ClydeCosmetics/Environments/KelpLinen/MF_KelpLinen_LocationFade}.Output))
  n11 = Subtract(A=CameraPositionWS(), B=ActorPositionWS())
  n12 = Divide(A=Subtract(A=Length(None=@n11) {is=class MaterialExpressionLength}, B=@n7), B=@n6)
  Multiply(A=Multiply(A=@n10, B=Saturate(None=@n12)), B=ScalarParameter() {defaultValue=1, parameterName=Phase Opacity Mod})
  ObjectPositionWS()
  ScalarParameter() {parameterName=Custom Fade Offset}
  ScalarParameter() {defaultValue=1, parameterName=custom Fade lenth}
  n13 = CollectionParameter() {collection=&FortniteMaterialParameters, parameterName=Main_Phase_Tracker}
  MaterialExpressionVectorParameter_3 = VectorParameter() {parameterName=Phase Opacity, channelNames {a=3, b=2, g=1, r=Empty}, defaultValue=<rgba 1 0.75 1 1>}
  Switch!(
   SwitchValue=@n13
   Default
   Empty=@MaterialExpressionVectorParameter_3.Empty
   Phase1=@MaterialExpressionVectorParameter_3.Empty
   Phase2=@MaterialExpressionVectorParameter_3.Empty
   Phase3=@MaterialExpressionVectorParameter_3.Empty
  ) {
   constDefault=1
   inputs [
    {
     inputName=Empty
     input {
      expression=&self:MaterialExpressionVectorParameter_3
      inputName=None
      mask=1
      maskA=0
      maskB=0
      maskG=0
      maskR=1
      outputIndex=1
     }
    }
    {
     inputName=Phase1
     input {
      expression=&self:MaterialExpressionVectorParameter_3
      inputName=None
      mask=1
      maskA=0
      maskB=0
      maskG=1
      maskR=0
      outputIndex=2
     }
    }
    {
     inputName=Phase2
     input {
      expression=&self:MaterialExpressionVectorParameter_3
      inputName=None
      mask=1
      maskA=0
      maskB=1
      maskG=0
      maskR=0
      outputIndex=3
     }
    }
    {
     inputName=Phase3
     input {
      expression=&self:MaterialExpressionVectorParameter_3
      inputName=None
      mask=1
      maskA=1
      maskB=0
      maskG=0
      maskR=0
      outputIndex=4
     }
    }
   ]
  }
  WorldPosition()
 ]
}
