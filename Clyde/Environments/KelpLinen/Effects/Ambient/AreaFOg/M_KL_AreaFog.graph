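# Translucent unlit material that renders ambient area fog as camera-facing smoke sprites.
# World position offset sizes sprites, collapsing them near the camera or by per-platform vertex-alpha thresholds, and adds sine-driven drift.
# Vertex colors vary a rotating 8x8 smoke flipbook and panning noise, which blends base color with hue-shifted color under time-of-day tints.
# Opacity combines camera fade, time-of-day multipliers and view-angle-dependent depth fade, omitted on low-end platforms like Switch.
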Material /Clyde/Environments/KelpLinen/Effects/Ambient/AreaFOg/M_KL_AreaFog
graph {
 properties {
  bAutomaticallySetUsageInEditor=false
  bUseMaterialAttributes=true
  blendMode=BLEND_Translucent
  shadingModel=MSM_Unlit
  usedShadingModels=Unlit
 }
 nodes [
  Constant() {r=1}
  n1 = WorldPosition()
  n2 = VertexColor()
  n3 = MaterialFunctionCall(
   "Xb1,PS4"=ScalarParameter() {group=AlphaThreshhold, parameterName="alpha Thresold Offset  Xb1,PS4 (S)"}
   "PC-High&&Epic,XSX,PS5"=ScalarParameter() {group=AlphaThreshhold, parameterName=alpha Thresold Offset PC Highend epic XSX SP%, sliderMax=0.5, sliderMin=-0.5}
   "PC-Perf,Android"=ScalarParameter() {group=AlphaThreshhold, parameterName=alpha Thresold Offset PC perf/ Android}
   Switch=ScalarParameter() {group=AlphaThreshhold, parameterName="alpha Thresold Offset Switch (S)"}
   "PC-Low(NotPerf)&&Medium"=ScalarParameter() {group=AlphaThreshhold, parameterName="alpha Thresold Offset PC-Low(NotPerf)&&Medium (S)"}
  ) {materialFunction=&PlatformSwitch_1D}
  n4 = Saturate(None=Round(None=Add(A=@n2.A, B=@n3.Result)))
  n5 = SphereMask(
   A=MaterialFunctionCall() {materialFunction=&BoundingBoxBased_0-1_UVW}.RGB
   B=Constant() {r=0.5}
   Radius=ScalarParameter() {defaultValue=500, parameterName=Phase Spheremask Sprite size}
   Hardness=Constant() {r=1}
  )
  n6 = If(
   A=Length(None=WorldPosition() {worldPositionShaderOffset=WPT_CameraRelativeNoOffsets}.XYZ) {is=class MaterialExpressionLength}
   B=ScalarParameter() {defaultValue=35000, parameterName=Cull Sprite closer than this disance}
   "A > B"=Multiply(A=ScalarParameter() {defaultValue=8, parameterName=Size}, B=Multiply(A=@n4, B=@n5))
   "A < B"=Constant()
  )
  n7 = VectorParameter() {parameterName=Muiliplky WPO, channelNames {a=NA, b=z, g=y, r=x}, defaultValue=<rgba 1 1 1 0>}
  n8 = Multiply(A=Time(), B=ScalarParameter() {defaultValue=0.001, parameterName=VarianceSpeed}) {constB=0.05}
  n9 = ComponentMask(None=TextureCoordinate() {coordinateIndex=1}) {r=true}
  n10 = Add(A=Multiply(A=Time(), B=ScalarParameter() {defaultValue=0.4, parameterName=DriftSpeed}), B=Sine(None=Add(A=@n8, B=@n9)) {period=0.17})
  n11 = Sine(None=@n10) {period=0.56}
  n12 = Sine(None=@n10) {period=0.7}
  n13 = ScalarParameter() {defaultValue=10, parameterName=MinHorizontalDrift}
  n14 = ScalarParameter() {defaultValue=100, parameterName=MaxHorizontalDrift}
  n15 = Sine(None=Add(A=@n8, B=@n9)) {period=0.14}
  n16 = ScalarParameter() {defaultValue=1, group=WPO Curl, parameterName=CurlMult}
  n17 = Multiply(
   A=AppendVector(A=AppendVector(A=@n11, B=@n12), B=Sine(None=@n10) {period=0.82})
   B=Multiply(A=LinearInterpolate(A=@n13, B=@n14, Alpha=@n15), B=Clamp(None=@n16) {maxDefault=2})
  )
  n18 = ScalarParameter() {defaultValue=0.6, parameterName=MinVerticalSpeedMult}
  n19 = Constant() {r=0.5}
  n20 = Multiply(A=ObjectRadius(), B=LinearInterpolate(A=Multiply(A=@n18, B=@n19), B=@n19, Alpha=ConstantBiasScale(None=@n15)) {constA=0.5, constB=0.5}) {constB=0.5}
  n21 = TextureCoordinate() {coordinateIndex=1}
  n22 = Multiply(A=Subtract(A=ComponentMask(None=@n21) {g=true}) {constB=0.5}) {constB=-2}
  n23 = Multiply(
   A=LinearInterpolate(A=Multiply(A=@n20) {constB=-1}, B=@n20, Alpha=Frac(None=@n22))
   B=VectorParameter() {parameterName=OverallDirection, defaultValue=<rgba 0 0 1 1>}.RGB
  )
  n24 = NamedRerouteDeclaration(None=Multiply(A=Add(A=@n17, B=@n23), B=VectorParameter() {parameterName=Offset Scale, defaultValue=<rgba 1 1 1 0>}.RGB)) {name=WPO, nodeColor=<rgba 1 0.8203 0 1>}
  n25 = Multiply(A=Time(), B=ScalarParameter() {parameterName=Hueshift Speed})
  n26 = VertexColor()
  n27 = Constant() {r=8}
  n28 = TextureCoordinate()
  n29 = ScalarParameter() {group=Cloud Tex, parameterName=Rotation SPeed Low}
  n30 = ScalarParameter() {group=Cloud Tex, parameterName=Rotation SPeed High}
  n31 = Add(A=@n26.G, B=Multiply(A=Time(), B=LinearInterpolate(A=@n29, B=@n30, Alpha=@n26.A)))
  n32 = TextureSample!(
   UVs=MaterialFunctionCall(
    "Animation  Phase (0-1)"=@n26.R
    Number of Rows=@n27
    Number of Columns=@n27
    UVs=MaterialFunctionCall(UVs=@n28, "Rotation Angle (0-1)"=@n31) {materialFunction=&CustomRotator}.Rotated Values
   ) {materialFunction=&FlipBook}.UVs
   Tex
   Bias=MaterialFunctionCall(
    "Xb1,PS4"=ScalarParameter() {group=Cloud Tex, parameterName="Cloud MipBias Xb1,PS4 (S)"}
    "PC-High&&Epic,XSX,PS5"=ScalarParameter() {group=Cloud Tex, parameterName="Cloud MipBias PC-High&&Epic,XSX,PS5 (S)"}
    "PC-Perf,Android"=ScalarParameter() {group=Cloud Tex, parameterName="Cloud MipBias PC-Perf,Android (S)"}
    Switch=ScalarParameter() {group=Cloud Tex, parameterName="Cloud MipBias Switch (S)"}
    "PC-Low(NotPerf)&&Medium"=ScalarParameter() {group=Cloud Tex, parameterName="Cloud MipBias PC-Low(NotPerf)&&Medium (S)"}
   ) {materialFunction=&PlatformSwitch_1D}.Result
   Apply View MipBias
  ) {mipValueMode=TMVM_MipBias, texture=/Game/Effects/Fort_Effects/Textures/Smoke/T_Smoke_Thick_8x8_01}
  n33 = Multiply(A=LinearInterpolate(Alpha=@n32.RGB) {constA=-0.5}, B=ScalarParameter() {parameterName=Hueshift Range})
  n34 = ScalarParameter() {defaultValue=1, parameterName=Per sprite offset varience}
  n35 = VertexColor()
  n36 = MaterialFunctionCall(Hue Shift Percentage=Add(A=Add(A=@n25, B=@n33), B=Multiply(A=@n34, B=@n35.B)), Texture=Constant3Vector() {constant=<rgba 0 0 1 0>}) {materialFunction=&HueShift}
  n37 = OneMinus(None=ScalarParameter() {parameterName=Hue shift saturatiuon})
  n38 = ScalarParameter() {defaultValue=0.1, parameterName=color Pan Speed}
  n39 = VertexColor()
  n40 = Normalize(VectorInput=Subtract(A=ComponentMask(None=@n39) {g=true, r=true}) {constB=0.5})
  n41 = TextureObject() {texture=/Game/Global/Textures/Default/Noise/T_Noise_Cloud}
  n42 = TextureSample!(UVs=Panner(Speed=Multiply(A=@n38, B=@n40)), Tex=@n41, Apply View MipBias) {texture=&T_Noise_Cloud}
  n43 = Normalize(VectorInput=Subtract(A=AppendVector(A=@n39.B, B=@n39.B)) {constB=0.5})
  n44 = TextureSample(UVs=Panner(Speed=Multiply(A=@n38, B=@n43)), Tex=@n41) {texture=&T_Noise_Cloud}
  n45 = LinearInterpolate(A=Min(A=@n42.R, B=@n44.G), B=Max(A=@n42.R, B=@n44.G), Alpha=ScalarParameter() {parameterName=Color Transition min vs max})
  n46 = VectorParameter() {parameterName=Color transition Mask Color Options, defaultValue=<rgba 1 1 0 0>}
  n47 = Add(A=Multiply(A=Power(Base=@n45, Exp=@n46.R) {constExponent=4}, B=@n46.G) {constB=7}, B=@n46.B)
  n48 = LinearInterpolate(
   A=VectorParameter() {parameterName=Mail Smoke Color, defaultValue=<rgba 0 0 1 0>}.RGB
   B=Multiply(A=Desaturation(None=@n36.Result, Fraction=@n37), B=ScalarParameter() {parameterName=Hueshift value})
   Alpha=Saturate(None=MaterialFunctionCall(In=@n47, Contrast=@n46.A) {materialFunction=&CheapContrast}.Result)
  )
  n49 = ScalarParameter() {defaultValue=1, parameterName=Emissive STR}
  n50 = VectorParameter() {group=TOD, parameterName=TOD Morning tint and opac mul, defaultValue=<rgba 1 1 1 1>}
  n51 = VectorParameter() {group=TOD, parameterName=TOD Day tint and opac mul, defaultValue=<rgba 1 1 1 1>}
  n52 = VectorParameter() {group=TOD, parameterName=TOD Evening tint and opac mul, defaultValue=<rgba 1 1 1 1>}
  n53 = VectorParameter() {group=TOD, parameterName=TOD Night tint and opac mul, defaultValue=<rgba 1 1 1 1>}
  n54 = ScalarParameter() {parameterName="Fade Length (S)"}
  n55 = ScalarParameter() {parameterName="Fade Offset (S)"}
  n56 = Normalize(VectorInput=Subtract(A=CameraPositionWS(), B=ActorPositionWS()))
  n57 = Constant3Vector() {constant=<rgba 0 0 1 0>}
  n58 = VectorParameter() {parameterName=dot fade options, defaultValue=<rgba 1 1 0 0>}
  n59 = Power(Base=Abs(None=DotProduct(A=@n56, B=@n57)), Exp=@n58.R)
  n60 = MaterialFunctionCall(In=Add(A=Multiply(A=@n59, B=@n58.G), B=@n58.B), Contrast=@n58.A) {materialFunction=&CheapContrast}
  n61 = Constant() {r=0.5}
  n62 = Multiply(A=@n32.R, B=SphereMask(A=TextureCoordinate(), B=@n61, Radius=@n61, Hardness=Constant() {r=0.75}))
  n63 = DepthFade(
   Opacity=@n62
   FadeDistance=LinearInterpolate(A=ScalarParameter() {parameterName=FadeDistanceSide}, B=ScalarParameter() {parameterName=FadeDistanceTop}, Alpha=Saturate(None=@n60.Result))
  )
  n64 = Multiply(
   A=MaterialFunctionCall(Value_Morning=@n50.A, Value_Day=@n51.A, Value_Evening=@n52.A, Value_Night=@n53.A) {materialFunction=&Scalar_Time_of_Day}.Result
   B=Multiply(
    A=MaterialFunctionCall("Xb1,PS4"=@n63, "PC-High&&Epic,XSX,PS5"=@n63, "PC-Perf,Android"=@n62, Switch=@n62, "PC-Low(NotPerf)&&Medium"=@n63) {materialFunction=&PlatformSwitch_1D}.Result
    B=MaterialFunctionCall(Fade Length=@n54, Fade Offset=@n55) {materialFunction=&CameraDepthFade}.Result
   )
  )
  n65 = ScalarParameter() {defaultValue=1, parameterName=Phase opacity mod}
  MakeMaterialAttributes!(
   BaseColor
   Metallic
   Specular
   Roughness
   Anisotropy
   EmissiveColor=Multiply(
    A=Multiply(A=@n48, B=@n49)
    B=MaterialFunctionCall(Color_Morning=@n50.RGB, Color_Day=@n51.RGB, Color_Evening=@n52.RGB, Color_Night=@n53.RGB) {materialFunction=&Color_Time_of_Day}.Result
   )
   Opacity=Multiply(A=Multiply(A=@n64, B=@n65), B=ScalarParameter() {defaultValue=1, parameterName=Master Opacity})
   OpacityMask
   Normal
   Tangent
   WorldPositionOffset=Add(
    A=MaterialFunctionCall!(Center Location=@n1.XYZ, 0-1 UVs, XY scale=@n6, OverridePivot, "Normals (Optional)") {materialFunction=&Sprite}.World position offset
    B=Multiply(A=@n7, B=@n24)
   )
   SubsurfaceColor
   ClearCoat
   ClearCoatRoughness
   AmbientOcclusion
   Refraction
   CustomizedUVs_0
   CustomizedUVs_1
   CustomizedUVs_2
   CustomizedUVs_3
   CustomizedUVs_4
   CustomizedUVs_5
   CustomizedUVs_6
   CustomizedUVs_7
   PixelDepthOffset
   ShadingModel
   Displacement
  )
  ScalarParameter() {parameterName=Mip Bias}
 ]
}
