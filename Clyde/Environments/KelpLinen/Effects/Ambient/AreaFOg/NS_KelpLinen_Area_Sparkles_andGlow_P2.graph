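# Ambient Niagara area effect for the KelpLinen environment, combining a blue glow mesh with drifting cyan-to-purple sparkles.
# Glow burst-spawns a looping mesh particle whose brightness pulses and varies with time of day.
# Sparkles spawns curl-noise-driven sprites in a squashed sphere, shrinking over their lifetime.
# Both emitters measure distance to the camera, gating glow mesh scale and sparkle spawn rate against thresholds.
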NiagaraSystem /Clyde/Environments/KelpLinen/Effects/Ambient/AreaFOg/NS_KelpLinen_Area_Sparkles_andGlow_P2
dependencies {
 usedDataInterfaces [&NiagaraDataInterfaceCurve, &NiagaraDataInterfaceArrayFloat, &NiagaraDataInterfaceCamera]
 usedDynamicInputs [
  &ConvertVectorToPosition
  &Multiply_Vector2D_ByFloat
  &ReturnCameraPosition
 ]
 usedModules [
  &SystemState
  &self:SystemSpawnScript.SystemScriptSource.SystemScriptGraph.NiagaraNodeAssignment_0.SetVariables_#1
  &self:SystemSpawnScript.SystemScriptSource.SystemScriptGraph.NiagaraNodeAssignment_1.SetVariables_#2
 ]
}
emitters {
 Glow {
  emitterUpdateScript [
   &EmitterState()
   &SpawnBurst_Instantaneous(Use Spawn Probability=false, UseLoopCountLimit=true)
   /SetVariables(Emitter.Dist:NiagaraFloat=&DistanceBetweenPositions(
     A=&TransformPosition(Position=/Niagara/DynamicInputs/Helpers/SimulationPosition(), Source Space=::World, Destination Space=::Simulation)
     B=$System.CamPos
    )) {isSetParameters=true}
  ]
  particleSpawnScript [&InitializeParticle(
    Color=<rgba 0.07 0.21 0.7 0.8>
    Lifetime=5
    Mesh Scale=&Multiply_VectorByFloat(Vector=<1.2 1.2 0.95>, Float=93.1661)
    Mesh Scale Mode=::NewEnumerator4
    Position Offset=&Multiply_VectorByFloat(Vector=<0 0 3>, Float=-1000)
    Sprite Rotation Mode=::NewEnumerator2
    UsePositionOffset=true
    Write Material Random=true
   )]
  particleUpdateScript [
   &ParticleState(
    Kill Particles When Lifetime Has Elapsed=false
    Let Infinitely Lived Particles Die When Emitter Deactivates=true
    Loop Particles Lifetime=true
   )
   &ScaleMeshSize(Scale Factor=&VectorFromFloat(Value=&MakeCustomFloatFromBool(Bool=&SetBoolByFloatComparison(A=$Emitter.Dist, B=35000), True Float=1, False Float=0)))
   &ScaleColor(
    Scale Alpha=&Lerp_Float(
     A=1
     B=1
     Alpha=&FloatFromCurve(FloatCurve=DataInterface() {propertyValues {Curve {keys [{arriveTangent=3.807, interpMode=RCIM_Cubic, time=0, value=0}, <at 0.5 1 RCIM_Cubic>, <at 1 0 RCIM_Cubic>]}}})
    )
    Scale RGB=&VectorFromFloat(Value=&Lerp_Float(
      A=&FNTimeOfDayScalar(Night Scale=0.4, Morning Scale=0.4, Day Scale=0.6, Evening Scale=0.5)
      B=&FNTimeOfDayScalar(Night Scale=0.5, Morning Scale=0.5, Day Scale=0.7, Evening Scale=0.6)
      Alpha=&FloatFromCurve(FloatCurve=DataInterface() {propertyValues {Curve {keys [{arriveTangent=3.807, interpMode=RCIM_Cubic, time=0, value=0}, <at 0.5 1 RCIM_Cubic>, <at 1 0 RCIM_Cubic>]}}})
     ))
    ScaleA=true
    ScaleRGB=true
   )
   &ScaleColor(
    Scale Alpha=&FloatFromCurve(
     FloatCurve=DataInterface() {propertyValues {Curve {keys [<at 0 0 RCIM_Cubic>, {tangentMode=RCTM_User, time=1, value=1}]}}}
     CurveIndex=&Divide_Float(A=$Emitter.Age, B=2.2)
    )
    ScaleA=true
    ScaleRGB=true
   ) {enabled=false, module=ScaleColor001}
   &ScaleColor(Scale RGB=&VectorFromFloat(Value=0.75), ScaleA=true, ScaleRGB=true) {module=ScaleColor002}
   &ScaleColor(ScaleA=true, ScaleRGB=true) {module=ScaleColor003}
  ]
  rendererClasses [&NiagaraMeshRendererProperties]
  renderers [{rendererClass=&NiagaraMeshRendererProperties, rendererIndex=0}]
 }
 Sparkles {
  emitterSpawnScript [/SetVariables(Emitter.DistanceKill:NiagaraFloat=&Multiply_Float(A=4.5, B=10000)) {isSetParameters=true}]
  emitterUpdateScript [
   &EmitterState(
    Delay First Loop Only=true
    Life Cycle Mode=::NewEnumerator1
    Loop Delay=0.4253
    Loop Duration=0.6522
    Recalculate Duration Each Loop=true
    UseLoopDelay=true
   )
   /SetVariables(Emitter.Dist:NiagaraFloat=&DistanceBetweenPositions(
     A=&TransformPosition(Position=&SimulationPosition(), Source Space=::World, Destination Space=::Simulation)
     B=$System.CamPos
    )) {isSetParameters=true}
   &SpawnRate(
    SpawnRate=&MakeCustomFloatFromBool(Bool=&SetBoolByFloatComparison(A=$Emitter.Dist, B=$Emitter.DistanceKill), True Float=7.5, False Float=0)
    Use Spawn Probability=false
   )
  ]
  particleSpawnScript [
   &KillParticles(Kill Particles=&SetBoolByFloatComparison(
     A=&Abs_Float(Float:NiagaraFloat=&DistanceBetweenPositions(A=$System.CamPos, B=$Particles.Position))
     B=$Emitter.DistanceKill
     Comparison Type=::NewEnumerator6
    )) {enabled=false}
   &InitializeParticle(
    Color=&RandomRangeLinearColor(Minimum=<rgba 0 0.8192 1 1>, Maximum=<rgba 0.6188 0 1 1>)
    Lifetime=&RandomRangeFloat(Minimum=2, Maximum=3.5)
    Mesh Scale Mode=::NewEnumerator3
    Position Offset=&Multiply_VectorByFloat(Vector=<2.5 4.5 2>, Float=1000)
    Sprite Rotation Mode=::NewEnumerator2
    Sprite Size Mode=::NewEnumerator3
    Uniform Sprite Size=&Multiply_Float(A=2.5, B=500)
    UsePositionOffset=true
    Write Material Random=true
   )
   &ShapeLocation(
    MaskSpawnGroup=false
    Non Uniform Scale=<1.3 1.1 0.5>
    Override Seed=false
    Sphere Radius=&Multiply_Float(A=10, B=1000)
    UseOwnerScale=true
   )
   &CameraOffset()
  ]
  particleUpdateScript [
   &ParticleState()
   &ScaleSpriteSize(Scale Factor=&Vector2DFromFloat(Value=&FloatFromCurve(FloatCurve=DataInterface() {propertyValues {Curve {keys [<at 0 1 RCIM_Cubic>, {arriveTangent=-1.6315, interpMode=RCIM_Cubic, leaveTangent=-1.6315, tangentMode=RCTM_User, time=1, value=0}]}}}))) {module=ScaleSpriteSize001}
   &ScaleSpriteSize(Uniform Curve Sprite Scale=DataInterface() {propertyValues {Curve {keys [
        <at 0 0 RCIM_Cubic>
        {arriveTangent=-3.0003, interpMode=RCIM_Cubic, leaveTangent=-3.0003, tangentMode=RCTM_User, time=0.1, value=1}
        <at 1 0 RCIM_Cubic>
       ]}}})
   &CurlNoiseForce(Noise Strength=1000, Pan Noise=false)
   &SolveForcesAndVelocity(Clamp Velocity=false, Limit Acceleration=false) {module=SolveForcesAndVelocity001}
   &ScaleColor(
    Scale Alpha=&FloatFromCurve(FloatCurve=DataInterface() {propertyValues {Curve {keys [<at 0 1 RCIM_Cubic>,<at 0.1 3 RCIM_Cubic>,<at 0.2 1 RCIM_Cubic>]}}})
    Scale RGB=&VectorFromFloat(Value=7.5)
    ScaleA=true
    ScaleRGB=true
   )
   &DynamicMaterialParameters(
    Index 0 Param 1=250
    Module.Param0WriteEnabled=true
    Module.Param1WriteEnabled=true
    Module.Param2WriteEnabled=true
    Module.Param3WriteEnabled=true
   )
  ]
  rendererClasses [&NiagaraSpriteRendererProperties]
  renderers [
   {rendererClass=&NiagaraSpriteRendererProperties}
   {rendererClass=&NiagaraSpriteRendererProperties, rendererIndex=1}
  ]
 }
}
summary {
 systemName=NS_KelpLinen_Area_Sparkles_andGlow_P2
 emitters [
  {emitterName=Sparkles, rendererClasses [&NiagaraSpriteRendererProperties]}
  {emitterName=Glow, rendererClasses [&NiagaraMeshRendererProperties]}
 ]
 userVariables [
  User.BuildupEnergy01 {type=NiagaraFloat, defaultValue=1, description=""}
  User.SlopeAngle {type=NiagaraFloat, defaultValue=0, description=""}
 ]
}
